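# Translucent, unlit, two-sided beam trail material that draws candy cane stripes, filed with the Gingerbread glider's effects.
# The candy cane texture pans along the second UV channel, rotated by a particle dynamic parameter, then is tinted by particle color and passed through emissive adjustments.
# A squared sphere mask on power-curved U, multiplied by a sine of V and particle alpha, forms a soft fade along and across the trail.
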Material /BRCosmetics/Gadgets/Assets/VinderTech_GliderChute/Glider_Gingerbread/FX/M_CandyCane_Trail
graph {
 properties {
  bUsedWithBeamTrails=true
  blendMode=BLEND_Translucent
  pixelDepthOffsetMode=PDOM_Legacy
  shadingModel=MSM_Unlit
  twoSided=true
  usedShadingModels=Unlit
 }
 nodes [
  n1 = TextureCoordinate() {coordinateIndex=1, uTiling=2}
  MaterialFunctionCall(XY=@n1) {materialFunction=&Swizzle}
  n2 = MaterialFunctionCall(
   UVs=Panner(Coordinate=@n1) {speedX=1}
   "Rotation Angle (0-1)"=DynamicParameter() {defaultValue=<rgba 0.125 1 1 1>, paramNames [Rotation,Param2,Param3,Param4]}.Rotation
  ) {materialFunction=&CustomRotator}
  n3 = ParticleColor()
  MaterialFunctionCall(Emissive=Multiply(A=TextureSample(UVs=@n2.Rotated Values) {texture=/Game/Environments/Props/Winter/Textures/T_CandyCane_D}.RGB, B=@n3.RGB)) {materialFunction=&MF_FX_EmissiveAdjustments}
  n4 = MaterialFunctionCall(Float2=TextureCoordinate()) {materialFunction=&BreakOutFloat2Components}
  n5 = SphereMask(A=Power(Base=@n4.R) {constExponent=0.15}, B=Constant() {r=0.55}) {attenuationRadius=0.5, hardnessPercent=0}
  Multiply(A=@n3.A, B=Multiply(A=Multiply(A=@n5, B=@n5), B=Sine(None=@n4.G) {period=2}))
 ]
}
